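; Race track actor for drivable vehicle races, presenting start lights, a timer screen and a three-slot scoreboard.
; The construction script creates dynamic material instances for these displays, resets their parameters and adds a box collision at each checkpoint transform.
; RepNotify functions update lap time, flag and top-three times, and play countdown beeps at the start light.
; Times become material digit parameters by chopping formatted time strings.
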(fn OnRep_RN-1stPlaceTime
  (CallFunction|SetScoreboardwithRacerTime (Variables|Default|GetRN-1stPlaceTime)))

(fn OnRep_RN-2ndPlaceTime
  (CallFunction|SetScoreboardwithRacerTime (Variables|Default|GetRN-2ndPlaceTime) 1))

(fn OnRep_RN-3rdPlaceTime
  (CallFunction|SetScoreboardwithRacerTime (Variables|Default|GetRN-3rdPlaceTime) 2))

(fn OnRep_RN-CountdownTimer
  (bind _s_racetrack_startlight (Variables|Default|GetS_Racetrack_StartLight))
  (bind _returnvalue (Transformation|GetWorldLocation _s_racetrack_startlight))
  (bind _rn_countdowntimer (Variables|Default|GetRN-CountdownTimer))
  (if (> _rn_countdowntimer 3)
    ;[1 1 1] Above 3 seconds
    (Audio|PlaySoundatLocation "/Game/Sounds/Fort_GamePlay_Sounds/Missions/Race_Beep_Count_Cue.Race_Beep_Count_Cue" _returnvalue)
    (else
      ;[1 1 1] 3-2-1-0
      (Audio|PlaySoundatLocation (Utilities|Select :Index _rn_countdowntimer :Option%200 "/Game/Sounds/Fort_GamePlay_Sounds/Missions/Race_Beep_Start_Cue.Race_Beep_Start_Cue" :Option%201 "/Game/Sounds/Fort_GamePlay_Sounds/Missions/Race_Beep_Ready_Cue.Race_Beep_Ready_Cue" :Option%202 "/Game/Sounds/Fort_GamePlay_Sounds/Missions/Race_Beep_Ready_Cue.Race_Beep_Ready_Cue" :Option%203 "/Game/Sounds/Fort_GamePlay_Sounds/Missions/Race_Beep_Ready_Cue.Race_Beep_Ready_Cue") _returnvalue))))

(fn OnRep_RN-SmallScreenLapTime
  (bind _rn_smallscreenlaptime (Variables|Default|GetRN-SmallScreenLapTime))
  (bind _returnvalue (Utilities|String|TimeSecondstoString _rn_smallscreenlaptime))
  (bind _returnvalue_2 (Utilities|String|LeftChop _returnvalue 3))
  (bind _mid_timer (Variables|Default|GetMID-Timer))
  (Rendering|Material|SetScalarParameterValue _mid_timer "TenMinute" (Utilities|String|StringToFloat (Utilities|String|LeftChop _returnvalue_2 4)) :!class /Script/Engine.MaterialInstanceDynamic)
  (Rendering|Material|SetScalarParameterValue _mid_timer "OneMinute" (Utilities|String|StringToFloat (Utilities|String|RightChop (Utilities|String|LeftChop _returnvalue_2 3) 1)) :!class /Script/Engine.MaterialInstanceDynamic)
  (Rendering|Material|SetScalarParameterValue _mid_timer "TenSecond" (Utilities|String|StringToFloat (Utilities|String|LeftChop (Utilities|String|RightChop _returnvalue_2 3) 1)) :!class /Script/Engine.MaterialInstanceDynamic)
  (Rendering|Material|SetScalarParameterValue _mid_timer "OneSecond" (Utilities|String|StringToFloat (Utilities|String|RightChop _returnvalue_2 4)) :!class /Script/Engine.MaterialInstanceDynamic))

(fn OnRep_RN-TimerFlag
  (Rendering|Material|SetScalarParameterValue (Variables|Default|GetMID-Timer) "Flag" (Variables|Default|GetRN-TimerFlag) :!class /Script/Engine.MaterialInstanceDynamic))

(fn OnRep_ShowSmallScreenLapTime
  (bind _mid_timer (Variables|Default|GetMID-Timer))
  (if (Variables|Default|GetShowSmallScreenLapTime)
    (Rendering|Material|SetScalarParameterValue _mid_timer "LapTime" 1.0 :!class /Script/Engine.MaterialInstanceDynamic)
    (Rendering|Material|SetScalarParameterValue _mid_timer "Slider" 0.0 :!class /Script/Engine.MaterialInstanceDynamic)
    (else
      (Rendering|Material|SetScalarParameterValue _mid_timer "LapTime" 0.0 :!class /Script/Engine.MaterialInstanceDynamic)
      (Rendering|Material|SetScalarParameterValue _mid_timer "Slider" 1.0 :!class /Script/Engine.MaterialInstanceDynamic))))

(fn ResetMaterialSettings
  (bind _mid_startlights (Variables|Default|GetMID-StartLights))
  (bind _mid_timer (Variables|Default|GetMID-Timer))
  (bind _mid_scoreboard1 (Variables|Default|GetMID-Scoreboard1))
  (Rendering|Material|SetScalarParameterValue _mid_scoreboard1 "Position" 1.0 :!class /Script/Engine.MaterialInstanceDynamic)
  (Rendering|Material|SetScalarParameterValue (Variables|Default|GetMID-Scoreboard2) "Position" 2.0 :!class /Script/Engine.MaterialInstanceDynamic)
  (Rendering|Material|SetScalarParameterValue (Variables|Default|GetMID-Scoreboard3) "Position" 3.0 :!class /Script/Engine.MaterialInstanceDynamic)
  (Rendering|Material|SetScalarParameterValue _mid_scoreboard1 "TenMinute" 0.0 :!class /Script/Engine.MaterialInstanceDynamic)
  (Rendering|Material|SetScalarParameterValue _mid_scoreboard1 "OneMinute" 0.0 :!class /Script/Engine.MaterialInstanceDynamic)
  (Rendering|Material|SetScalarParameterValue _mid_scoreboard1 "TenSecond" 0.0 :!class /Script/Engine.MaterialInstanceDynamic)
  (Rendering|Material|SetScalarParameterValue _mid_scoreboard1 "OneSecond" 0.0 :!class /Script/Engine.MaterialInstanceDynamic)
  (Rendering|Material|SetScalarParameterValue _mid_timer "DigitLoc" 0.5 :!class /Script/Engine.MaterialInstanceDynamic)
  (Rendering|Material|SetScalarParameterValue _mid_timer "Use2Digits" 1.0 :!class /Script/Engine.MaterialInstanceDynamic)
  (Rendering|Material|SetScalarParameterValue _mid_timer "DigitA" 1.0 :!class /Script/Engine.MaterialInstanceDynamic)
  (Rendering|Material|SetScalarParameterValue _mid_timer "DigitB" 5.0 :!class /Script/Engine.MaterialInstanceDynamic)
  (Rendering|Material|SetScalarParameterValue _mid_timer "Brightness" 0.0 :!class /Script/Engine.MaterialInstanceDynamic)
  (Rendering|Material|SetScalarParameterValue _mid_timer "Arrows" 0.0 :!class /Script/Engine.MaterialInstanceDynamic)
  (Rendering|Material|SetScalarParameterValue _mid_timer "FlashControl" 0.0 :!class /Script/Engine.MaterialInstanceDynamic)
  (Rendering|Material|SetScalarParameterValue _mid_timer "Flash" 0.0 :!class /Script/Engine.MaterialInstanceDynamic)
  (Rendering|Material|SetScalarParameterValue _mid_startlights "L1" 0.0 :!class /Script/Engine.MaterialInstanceDynamic)
  (Rendering|Material|SetScalarParameterValue _mid_startlights "L2" 0.0 :!class /Script/Engine.MaterialInstanceDynamic)
  (Rendering|Material|SetScalarParameterValue _mid_startlights "L3" 0.0 :!class /Script/Engine.MaterialInstanceDynamic)
  (Rendering|Material|SetScalarParameterValue _mid_startlights "L4" 0.0 :!class /Script/Engine.MaterialInstanceDynamic)
  (Rendering|Material|SetScalarParameterValue _mid_startlights "L4-FalseStart" 0.0 :!class /Script/Engine.MaterialInstanceDynamic)
  (Rendering|Material|SetScalarParameterValue _mid_startlights "L5" 0.0 :!class /Script/Engine.MaterialInstanceDynamic))

(fn SetScoreboardWithRacerTime
  (params (TrackTimeInput :type Float%20%28double-precision%29) (RacerIndexInput :type Integer))
  (bind _returnvalue (Utilities|String|TimeSecondstoString TrackTimeInput))
  (bind _returnvalue_2 (Utilities|String|LeftChop _returnvalue 3))
  (bind _mid_scoreboard3 (Variables|Default|GetMID-Scoreboard3))
  (bind _mid_scoreboard2 (Variables|Default|GetMID-Scoreboard2))
  (bind _mid_scoreboard1 (Variables|Default|GetMID-Scoreboard1))
  (bind _array (Utilities|Array|MakeArray _mid_scoreboard1 _mid_scoreboard2 _mid_scoreboard3))
  (bind _output (Utilities|Array|Get(acopy) _array RacerIndexInput))
  (Rendering|Material|SetScalarParameterValue _output "Position" (Utilities|Array|Get(acopy) (Variables|Default|GetPlacement) RacerIndexInput) :!class /Script/Engine.MaterialInstanceDynamic)
  (Rendering|Material|SetScalarParameterValue _output "TenMinute" (Utilities|String|StringToFloat (Utilities|String|LeftChop _returnvalue_2 4)) :!class /Script/Engine.MaterialInstanceDynamic)
  (Rendering|Material|SetScalarParameterValue _output "OneMinute" (Utilities|String|StringToFloat (Utilities|String|RightChop (Utilities|String|LeftChop _returnvalue_2 3) 1)) :!class /Script/Engine.MaterialInstanceDynamic)
  (Rendering|Material|SetScalarParameterValue _output "TenSecond" (Utilities|String|StringToFloat (Utilities|String|LeftChop (Utilities|String|RightChop _returnvalue_2 3) 1)) :!class /Script/Engine.MaterialInstanceDynamic)
  (Rendering|Material|SetScalarParameterValue _output "OneSecond" (Utilities|String|StringToFloat (Utilities|String|RightChop _returnvalue_2 4)) :!class /Script/Engine.MaterialInstanceDynamic))

(fn ConstructionScript
  (bind _s_racetrack_startlight_scoreboard (Variables|Default|GetS_Racetrack_StartLight_Scoreboard))
  (bind _returnvalue (Rendering|Material|CreateDynamicMaterialInstance (Variables|Default|GetS_Racetrack_StartLight_Lights) 0 "/Game/Environments/Sets/RaceTrack/Materials/M_Racetrack_StartingLight_Detached.M_Racetrack_StartingLight_Detached" "None" :!class /Script/Engine.PrimitiveComponent))
  (Variables|Default|SetMID-StartLights _returnvalue)
  (bind _returnvalue_2 (Rendering|Material|CreateDynamicMaterialInstance (Variables|Default|GetS_Racetrack_StartLight_Timer) 0 "/Game/Environments/Sets/RaceTrack/Materials/M_Racetrack_StartLight_Timer.M_Racetrack_StartLight_Timer" "None" :!class /Script/Engine.PrimitiveComponent))
  (Variables|Default|SetMID-Timer _returnvalue_2)
  (bind _returnvalue_3 (Rendering|Material|CreateDynamicMaterialInstance _s_racetrack_startlight_scoreboard 1 0 "None" :!class /Script/Engine.PrimitiveComponent))
  (Variables|Default|SetMID-Scoreboard1 _returnvalue_3)
  (bind _returnvalue_4 (Rendering|Material|CreateDynamicMaterialInstance _s_racetrack_startlight_scoreboard 2 0 "None" :!class /Script/Engine.PrimitiveComponent))
  (Variables|Default|SetMID-Scoreboard2 _returnvalue_4)
  (bind _returnvalue_5 (Rendering|Material|CreateDynamicMaterialInstance _s_racetrack_startlight_scoreboard 3 0 "None" :!class /Script/Engine.PrimitiveComponent))
  (Variables|Default|SetMID-Scoreboard3 _returnvalue_5)
  ;[1 0.5 0] Set defaults for all MIDs
  (CallFunction|ResetMaterialSettings)
  ;[1 1 1] Add collision for each checkpoint
  (for (_array_element _array_index) (Variables|Default|GetCheckpointLocations)
    (AddComponent|NODE_AddBoxComponent-0 :RelativeTransform _array_element)))
