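; Race track checkpoint actor for BP_RaceTrack that records racers, or passengers of tag-matched vehicles, passing through it during a race.
; Where cosmetic events cannot run, overlaps validate checkpoint order, flag false starts, time finishes from green light or line crossing, and grant drivers fuel.
; It reports results to the track and toggles tube visuals for locally controlled pawns, with pass effects and sound.
; ResetCheckpoint enables collision, or clears racer lists and disables it.
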(event EventBeginPlay)

(event Collision|EventActorBeginOverlap
  (params (OtherActor :type Actor%20Object%20Reference)))

(event EventTick
  (params (DeltaSeconds :type Float%20%28single-precision%29)))

(event Custom|ResetCheckpoint
  (params (EnableCollision :type Boolean))
  (if EnableCollision
    ;[1 1 1] Turn ON collision
    (Collision|SetCollisionEnabled (Variables|Default|GetCollisionExtents) "QueryOnly")
    (else
      ;[1 1 1] Reset-Disable checkpoint
      (Utilities|Array|Clear (Variables|Default|GetDisqualifiedRacers))
      (Utilities|Array|Clear (Variables|Default|GetFinishedRacers))
      (Utilities|Array|Clear (Variables|Default|GetRacersThroughCheckpoint))
      (Utilities|Array|Clear (Variables|Default|GetRacerTotalTime))
      (Utilities|Array|Clear (Variables|Default|GetRacerStartTime))
      (Rendering|SetVisibility (Variables|Default|GetSM_HalfTube_Thin) false false)
      ;[1 1 1] Turn off collision
      (Collision|SetCollisionEnabled (Variables|Default|GetCollisionExtents) "NoCollision"))))

(event OnComponentBeginOverlap(CollisionExtents)
  (params (OverlappedComponent :type Primitive%20Component%20Object%20Reference) (OtherActor :type Actor%20Object%20Reference) (OtherComp :type Primitive%20Component%20Object%20Reference) (OtherBodyIndex :type Integer) (bFromSweep :type Boolean) (SweepResult :type Hit%20Result%20Structure%20%28by%20ref%29))
  (bind _self self)
  (bind _racetrackbp (Variables|Default|GetRaceTrackBP))
  (bind _allcheckpoints (Variables|/Game/Athena/DrivableVehicles/RaceTrack/BP_RaceTrack.BP_RaceTrack_C|GetAllCheckpoints _racetrackbp))
  (bind _returnvalue (Utilities|Array|FindItem _allcheckpoints _self))
  (bind _returnvalue_2 (+ 1 _returnvalue))
  (bind _finishedracers (Variables|Default|GetFinishedRacers))
  (bind _racetrackbp_2 (Variables|Default|GetRaceTrackBP))
  (bind _currentpawnbeingchecked (Variables|Default|GetCurrentPawnBeingChecked))
  (bind _currentpawnbeingchecked_2 (Variables|Default|GetCurrentPawnBeingChecked))
  (Networking|CanExecuteCosmeticEvents
    (:False
      (Utilities|String|LogString "Checkpoint pased")
      (if (or (Variables|/Game/Athena/DrivableVehicles/RaceTrack/BP_RaceTrack.BP_RaceTrack_C|GetPreparingRace (Variables|Default|GetRaceTrackBP)) (Variables|/Game/Athena/DrivableVehicles/RaceTrack/BP_RaceTrack.BP_RaceTrack_C|GetRaceInProgress (Variables|Default|GetRaceTrackBP)))
        (if (Variables|/Game/Athena/DrivableVehicles/RaceTrack/BP_RaceTrack.BP_RaceTrack_C|GetRequireVehicle (Variables|Default|GetRaceTrackBP))
          (bind _asfort_vehicle_interface (Utilities|Casting|CastToFortVehicleInterface OtherActor))
          (if (GameplayTags|HasAnyTags (Class|FortVehicleInterface|GetVehicleTags _asfort_vehicle_interface) (Variables|/Game/Athena/DrivableVehicles/RaceTrack/BP_RaceTrack.BP_RaceTrack_C|GetRequiredVehicleTag (Variables|Default|GetRaceTrackBP)))
            (for _array_element (Vehicle|GetAllPassengers _asfort_vehicle_interface)
              (Variables|Default|SetCurrentPawnBeingChecked _array_element)
              ;[1 1 1] Disable triggers for racers that are disqualified
              (if (Utilities|Array|ContainsItem (Variables|Default|GetDisqualifiedRacers) _currentpawnbeingchecked)
                (CallFunction|UpdateCheckpointFX self false _currentpawnbeingchecked)
                (elif (Utilities|Array|ContainsItem (Variables|Default|GetFinishedRacers) _currentpawnbeingchecked)
                  _
                  (elif (Variables|Default|GetFirstCP)
                    ;[1 1 1] If Racer goes through the fist CP again, and his name is on the list from the first time through, and he went through the previous CP, then he has finished the race
                    ;[1 1 1] False= first time through "FirstCP"
                    (if (Utilities|Array|ContainsItem (Variables|Default|GetRacersThroughCheckpoint) _currentpawnbeingchecked)
                      (if (Utilities|Array|ContainsItem (Variables|Default|GetRacersThroughCheckpoint (Variables|Default|GetFinalCheckpointBP-BeforeFinish)) _currentpawnbeingchecked)
                        ;[1 0.7652 0.1887 0.5] Only Add Drivers to list of finished Racers
                        (if (Variables|/Game/Athena/DrivableVehicles/RaceTrack/BP_RaceTrack.BP_RaceTrack_C|GetRequireVehicle (Variables|Default|GetRaceTrackBP))
                          (if (== (Vehicle|GetDriver _asfort_vehicle_interface) _currentpawnbeingchecked)
                            _))
                        ;[1 1 1] Create a list of racers that have finished
                        (bind _returnvalue_3 (Utilities|Array|AddUnique (Variables|Default|GetFinishedRacers) _currentpawnbeingchecked))
                        ;[0 0.5 0] Use the green light as the racer's start time, or use whenever the racer crosses the start line as his personal start time (drag race vs laps)
                        (if (Variables|Default|GetUseGreenlightAsStartTime)
                          (Class|BPRaceTrack|RacerFinished _racetrackbp_2 _returnvalue_3 (- (Utilities|Time|GetGameTimeinSeconds) (Variables|/Game/Athena/DrivableVehicles/RaceTrack/BP_RaceTrack.BP_RaceTrack_C|GetRaceStartTime (Variables|Default|GetRaceTrackBP))) _finishedracers)
                          (else
                            (Class|BPRaceTrack|RacerFinished _racetrackbp_2 _returnvalue_3 (- (Utilities|Time|GetGameTimeinSeconds) (Utilities|Array|Get(acopy) (Variables|Default|GetRacerStartTime) (Utilities|Array|FindItem (Variables|Default|GetRacersThroughCheckpoint) _currentpawnbeingchecked))) _finishedracers))))
                      (elif (Variables|/Game/Athena/DrivableVehicles/RaceTrack/BP_RaceTrack.BP_RaceTrack_C|GetPreparingRace (Variables|Default|GetRaceTrackBP))
                        ;[1 1 1] Disqualify racer if they go over the start line too soon
                        (Utilities|Array|AddUnique (Variables|Default|GetDisqualifiedRacers))
                        (Class|BPRaceTrack|FalseStart (Variables|Default|GetRaceTrackBP) 0)
                        (else
                          ;[1 1 1] Create list of racers that have entered the 1st CP
                          (bind _returnvalue_4 (Utilities|Array|AddUnique (Variables|Default|GetRacersThroughCheckpoint) _currentpawnbeingchecked))
                          ;[0 0.5 0] Use the green light as the racer's start time, or use whenever the racer crosses the start line as his personal start time (drag race vs laps)
                          (if (not (Variables|Default|GetUseGreenlightAsStartTime))
                            ;[1 1 1] Create a list of racer start times
                            (Utilities|Array|SetArrayElem (Variables|Default|GetRacerStartTime) _returnvalue_4 (Utilities|Time|GetGameTimeinSeconds) true))
                          ;[1 1 1] Skip using the CP death FX if 1st CP
                          (if (not (Variables|Default|GetFirstCP))
                            ;[1 1 1] Turn off the current CP FX
                            (CallFunction|UpdateCheckpointFX self false _currentpawnbeingchecked))
                          ;[1 1 1] Turn on the next sequential CP visuals
                          ;;[1 1 1] FORT-158220 //
                          ;; Don't go to the FX call if we are trying to access an out of bound index.
                          (if (not (>= _returnvalue_2 (Utilities|Array|Length _allcheckpoints)))
                            (CallFunction|UpdateCheckpointFX (Utilities|Array|Get(acopy) _allcheckpoints _returnvalue_2) true _currentpawnbeingchecked)))))
                    (bind _vehicle_occupied (Utilities|IsinVehicle _currentpawnbeingchecked_2)
                      (:true
                        (if (Vehicle|IsDrivingVehicle _currentpawnbeingchecked_2)
                          (Vehicle|Fuel|SetFuel _vehicle_occupied (+ (Vehicle|Fuel|GetFuel _vehicle_occupied) (ScalableFloat|GetValueatLevel (Variables|Default|GetGrantFuelToVehicleScalableFloat)))) ; Grant Vehicle More Fuel
                          (else
                            _))
                        (Class|BPRaceTrack|CheckpointUpdated (Variables|Default|GetRaceTrackBP)))
                      (:false
                        _))
                    (else
                      ;[1 1 1] Check to see if Racer went through the previous CP
                      (if (Utilities|Array|ContainsItem (Variables|Default|GetRacersThroughCheckpoint (Variables|Default|GetPreviousCheckpointBP)) _currentpawnbeingchecked)
                        ;[1 1 1] Disable trigger from racer if already went through this CP
                        (if (not (Utilities|Array|ContainsItem (Variables|Default|GetRacersThroughCheckpoint) _currentpawnbeingchecked))
                          ;[1 1 1] If Racer went through the previous CP, then add him to the list that went through this CP
                          (Utilities|Array|AddUnique (Variables|Default|GetRacersThroughCheckpoint) _currentpawnbeingchecked)))))))))
          (else
            (bind _asfort_player_pawn (Utilities|Casting|CastToFortPlayerPawn OtherActor))
            (if (not (Variables|/Game/Athena/DrivableVehicles/RaceTrack/BP_RaceTrack.BP_RaceTrack_C|GetRequireVehicle (Variables|Default|GetRaceTrackBP)))
              (Variables|Default|SetCurrentPawnBeingChecked _asfort_player_pawn))))))))

(event Custom|UpdateCheckpointFX
  (params (FX-Visibility :type Boolean) (FPP-01 :type Fort%20Player%20Pawn%20Object%20Reference))
  ;[1 1 1] CP FX control
  (if (Pawn|IsLocallyControlled FPP-01)
    (Rendering|SetVisibility (Variables|Default|GetSM_HalfTube_Thin) FX-Visibility false)
    (if (not FX-Visibility)
      (Effects|SpawnEmitterAttached :EmitterTemplate "/Game/Effects/Fort_Effects/Gameplay/CheckPoint/P_CheckPoint_Pass_BluePrint_FullCircle.P_CheckPoint_Pass_BluePrint_FullCircle" :AttachToComponent (Variables|Default|GetCheckPointParticles) :AttachPointName "None" :Scale "1.000000,1.000000,1.000000" :LocationType "SnapToTargetIncludingScale" :bAutoDestroy true :PoolingMethod "None" :bAutoActivate true)
      ;[1 0.5 0] Checkpoint sound should go here
      (Audio|PlaySound2D "/Game/Sounds/Fort_GamePlay_Sounds/Missions/Race_Checkpoint_Cue.Race_Checkpoint_Cue"))))

(fn ConstructionScript
  ;[1 1 1] Turn off collision
  (Collision|SetCollisionEnabled (Variables|Default|GetCollisionExtents) "NoCollision"))
